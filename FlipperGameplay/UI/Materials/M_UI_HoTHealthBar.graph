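# Translucent UI material for the HoT (likely heal-over-time) health bar, drawing a tinted, horizontally tapered shape.
# An edge mask combines a hardened right-end falloff with summed horizontal and vertical ramps from offset 9-slice UVs.
# A power-shaped gradient along normalized U multiplies that mask through scalar dot products, then FinalAlpha scales it.
# Color, edge hardness, offsets and taper are exposed as parameters.
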Material /FlipperGameplay/UI/Materials/M_UI_HoTHealthBar
graph {
 properties {blendMode=BLEND_Translucent, materialDomain=MD_UI, pixelDepthOffsetMode=PDOM_Legacy, usedShadingModels=Default Lit}
 nodes [
  n1 = MaterialFunctionCall() {materialFunction=&GetUserInterfaceUV}
  n2 = Multiply(A=ComponentMask(None=OneMinus(None=@n1.9-Slice UV)) {r=true}, B=ScalarParameter() {defaultValue=25, parameterName=EdgeHardnessEnd}) {constB=10}
  n3 = MaterialFunctionCall(Float2=Add(
    A=MaterialFunctionCall() {materialFunction=&GetUserInterfaceUV}.9-Slice UV
    B=ScalarParameter() {defaultValue=-0.5, parameterName=OffsetRight}
   )) {materialFunction=&BreakOutFloat2Components}
  n4 = Add(
   A=Multiply(A=Saturate(None=@n3.R), B=ScalarParameter() {defaultValue=10, parameterName=EdgeHardnessRightSide})
   B=Multiply(A=Saturate(None=@n3.G), B=ScalarParameter() {defaultValue=50, parameterName=EdgeHardness})
  )
  n5 = Saturate(None=DotProduct(A=Saturate(None=@n2), B=Saturate(None=@n4)))
  n6 = ScalarParameter() {defaultValue=-0.5, parameterName=GradiantOffset}
  n7 = ComponentMask(None=Saturate(None=MaterialFunctionCall() {materialFunction=&GetUserInterfaceUV}.Normalized UV)) {r=true}
  n8 = Power(Base=Saturate(None=LinearInterpolate(A=@n6, Alpha=@n7)), Exp=ScalarParameter() {defaultValue=1.25, parameterName=Taper})
  Multiply(A=Saturate(None=DotProduct(A=@n5, B=@n8)), B=ScalarParameter() {defaultValue=1, parameterName=FinalAlpha})
  VectorParameter() {parameterName=Color, defaultValue=<rgba 1 1 1 1>}
 ]
}
